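# Translucent unlit material for the Mystic gadget animation prop, usable on skeletal meshes, producing an animated flame-like glow.
# A Color parameter, HDR orange by default, tints a flame noise texture scrolled diagonally by adding Time to the UVs.
# A second tiling noise, stretched and scrolled along V, is multiplied by Color alpha and OpacityMult to form an opacity mask.
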Material /BRCosmetics/Animation/Props/Gadgets_Mystic/Materials/M_Gadget_Mystic_Prop
graph {
 properties {
  bUsedWithMeshDeformer=true
  bUsedWithSkeletalMesh=true
  blendMode=BLEND_Translucent
  pixelDepthOffsetMode=PDOM_Legacy
  shadingModel=MSM_Unlit
  usedShadingModels=Unlit
 }
 nodes [
  n1 = VectorParameter() {parameterName=Color, defaultValue=<rgba 20 1.6566 0 1>}
  n2 = Time()
  n3 = TextureCoordinate()
  Multiply(A=@n1.RGB, B=TextureSample(UVs=Add(A=@n2, B=@n3)) {texture=/Game/Effects/FogVolume/noise_flamy_Tile}.RGB)
  n4 = Subtract(A=Multiply(A=@n3, B=Constant2Vector() {g=0.2, r=1}), B=Multiply(A=@n2, B=Constant2Vector() {g=1}))
  Multiply(
   A=Multiply(A=@n1.A, B=TextureSample(UVs=@n4) {texture=/Game/Effects/FogVolume/T_TileNoise}.RGB)
   B=ScalarParameter() {defaultValue=1, parameterName=OpacityMult}
  )
 ]
}
